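# Parent character material for the PiteousKicks cape cosmetic, supporting clothing and skeletal meshes with the shading model set by expression.
# The 2023 character PreFX function feeds a cape-specific function, MF_PiteousKicks_Cape from the Pine skin folder, which also receives the Tech Art Mask.
# The UseReactivity scalar, grouped under PiteousKicks_GoldMaterial and on by default, blends base attributes toward that function's result.
# The blended attributes then enter the shared PostFX function.
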Material /BRCosmetics/Accessories/FORT_Capes/Cape_M_MED_PiteousKicks/Materials/M_PiteousKicks_Cape_PARENT
graph {
 properties {
  bAutomaticallySetUsageInEditor=false
  bUseMaterialAttributes=true
  bUsedWithClothing=true
  bUsedWithInstancedStaticMeshes=true
  bUsedWithMeshDeformer=true
  bUsedWithNanite=true
  bUsedWithNiagaraMeshParticles=true
  bUsedWithSkeletalMesh=true
  pixelDepthOffsetMode=PDOM_Legacy
  previewMesh=/CosmeticCompanions/Assets/Quadruped/CosmoHound_Meteor/Meshes/CosmoHound_Meteor_Quadruped
  shadingModel=MSM_FromMaterialExpression
  usedShadingModels=Default Lit | Subsurface | Clear Coat | Subsurface Profile | Cloth
 }
 nodes [
  n1 = MaterialFunctionCall() {materialFunction=&MF_CharacterParent_2023_PreFX}
  n2 = GetMaterialAttributes(None=@n1.Result) {attributeGetTypes [#1]}
  MaterialFunctionCall(
   Material=BlendMaterialAttributes(
    A=@n2.MaterialAttributes
    B=MaterialFunctionCall!(Tech Art Mask=@n1.Tech Art Mask, MaterialAttributes=@n2.MaterialAttributes, OriginalNormal=@n2.MaterialAttributes) {materialFunction=/Game/Accessories/FORT_Capes/Cape_M_MED_PiteousKicks/Skins/Pine/Materials/MF_PiteousKicks_Cape}.Result
    Alpha=ScalarParameter() {defaultValue=1, group=PiteousKicks_GoldMaterial, parameterName=UseReactivity}
   )
   M=@n1.M
   Pre Skinned Local Position=@n1.M
   Distance Blend=@n1.M
   Fresnel=@n1.M
  ) {materialFunction=&MF_CharacterParent_2023_PostFX}
 ]
}
